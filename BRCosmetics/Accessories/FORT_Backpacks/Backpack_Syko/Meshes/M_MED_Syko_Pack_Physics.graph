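# Physics asset for the Syko backpack cosmetic, defining bodies and constraints for its dangling pack segments.
# A chain of box bodies, dyn_pack_a through dyn_pack_f, hangs from a small base capsule, linked by limited swing and twist constraints that are tighter at the top.
# The base and pack capsules attach toward root with free angular joints.
# Capsules on pelvis, spine_02 and thighs have no constraints, likely serving as body colliders for the chain.
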PhysicsAsset /BRCosmetics/Accessories/FORT_Backpacks/Backpack_Syko/Meshes/M_MED_Syko_Pack_Physics
bodies {
 dyn_pack_a [{shapeType=Box, extentX=4.2581, extentY=14.371, extentZ=15.3042, center=<1.0684 0 0>, rotation=<rot 0 90 0>}]
 dyn_pack_a_base [{shapeType=Capsule, length=1, radius=1}]
 dyn_pack_b [{shapeType=Box, extentX=3.7559, extentY=8.6995, extentZ=10.8946, center=<2.8888 0 0>, rotation=<rot 0 90 0>}]
 dyn_pack_c [{shapeType=Box, extentX=3.9143, extentY=10.2466, extentZ=11.1252, center=<2.0569 0 0>, rotation=<rot 0 90 0>}]
 dyn_pack_d [{shapeType=Box, extentX=4.4387, extentY=10.8821, extentZ=11.2628, center=<3.272 0.0001 0>, rotation=<rot 0 90 0>}]
 dyn_pack_e [{shapeType=Box, extentX=3.4577, extentY=13.5518, extentZ=10.502, center=<4.4213 0 0>, rotation=<rot 0 90 0>}]
 dyn_pack_f [{shapeType=Box, extentX=2.4989, extentY=18.2539, extentZ=9.2154, center=<6.1544 0 0>, rotation=<rot 0 90 0>}]
 pack [{shapeType=Capsule, length=50.75, radius=27, center=<-0.0007 26.7594 -54.8625>}]
 pelvis [{shapeType=Capsule, length=9.5, radius=15, center=<-3.4306 1.6233 -0.0001>}]
 root [{shapeType=Capsule, length=0.505, radius=0.505, rotation=<rot 0 -90 0>}]
 spine_02 [{shapeType=Capsule, length=8.75, radius=11.25, rotation=<rot 90 0 -3.7007>}]
 thigh_l [{shapeType=Capsule, length=23.75, radius=10, center=<-18.2568 -0.4202 -1.7113>, rotation=<rot -86.5592 179.9999 -180>}]
 thigh_r [{shapeType=Capsule, length=20.25, radius=9.75, center=<18.6941 -0.0379 1.5725>, rotation=<rot -85.642 179.466 178.8877>}]
}
constraints [
 {bone1Name=dyn_pack_a, bone2Name=dyn_pack_a_base, swing1LimitDegrees=15, swing2LimitDegrees=5, twistLimitDegrees=5}
 {bone1Name=dyn_pack_b, bone2Name=dyn_pack_a, swing1LimitDegrees=50, swing2LimitDegrees=25, twistLimitDegrees=10}
 {bone1Name=dyn_pack_c, bone2Name=dyn_pack_b, swing1LimitDegrees=50, swing2LimitDegrees=25, twistLimitDegrees=10}
 {bone1Name=dyn_pack_d, bone2Name=dyn_pack_c, swing1LimitDegrees=50, swing2LimitDegrees=25, twistLimitDegrees=10}
 {bone1Name=dyn_pack_e, bone2Name=dyn_pack_d, swing1LimitDegrees=50, swing2LimitDegrees=25, twistLimitDegrees=10}
 {bone1Name=dyn_pack_f, bone2Name=dyn_pack_e, swing1LimitDegrees=50, swing2LimitDegrees=25, twistLimitDegrees=10}
 {
  bone1Name=dyn_pack_a_base
  bone2Name=pack
  swing1LimitDegrees=45
  swing1Motion=Free
  swing2LimitDegrees=45
  swing2Motion=Free
  twistLimitDegrees=45
  twistMotion=Free
 }
 {
  bone1Name=pack
  bone2Name=root
  swing1LimitDegrees=45
  swing1Motion=Free
  swing2LimitDegrees=45
  swing2Motion=Free
  twistLimitDegrees=45
  twistMotion=Free
 }
]
