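# Post-process style AnimBlueprint for the Syko backpack cosmetic, adding floating motion and secondary dynamics to its bones.
# Sine-driven offsets, pelvis, spine and thigh transforms copied from the parent mesh, and spine-driven root roll form a cached pose.
# That pose feeds RigidBody simulation, or a trail controller when rigid bodies are disabled or LOD exceeds 1, with trail strength varying by movement direction and skydiving.
# Torso twist/bend bone nodes are disconnected from the output.
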AnimBlueprint /BRCosmetics/Accessories/FORT_Backpacks/Backpack_Syko/Meshes/M_MED_Syko_Pack_AnimBP
target_skeleton=&/M_MED_Syko_Pack_Skeleton
graphs [
 AnimGraph {nodes [
   n1 = Get torsoBend() -> (torsoBend:double)
   "Transform (Modify) Bone - Bone: dyn_pack_a_base"(Alpha=@n1, Translation="0,0,0", Rotation="0,-30,0", Scale="1,1,1") {
    rotationMode=BMM_Additive
    rotationSpace=BCS_BoneSpace
    scaleMode=BMM_Additive
    scaleSpace=BCS_BoneSpace
    translationMode=BMM_Additive
    translationSpace=BCS_BoneSpace
    alphaBoolBlend {blendOption=Linear}
    alphaScaleBiasClamp {bMapRange=true, inRange {max=40}}
    boneToModify {boneName=dyn_pack_a_base}
    rotation {yaw=-30}
   }
   n2 = Get torsoTwist() -> (torsoTwist:double)
   n3 = "Transform (Modify) Bone - Bone: root"(Alpha=@n2, Translation="0,0,0", Rotation="-20,0,0") {
    rotationMode=BMM_Additive
    rotationSpace=BCS_BoneSpace
    scaleMode=BMM_Additive
    scaleSpace=BCS_BoneSpace
    translationMode=BMM_Additive
    translationSpace=BCS_BoneSpace
    alphaBoolBlend {blendOption=Linear}
    alphaScaleBiasClamp {bMapRange=true, inRange {max=40}}
    boneToModify {boneName=root}
    rotation {pitch=-20}
   }
   "Transform (Modify) Bone - Bone: root"(ComponentPose=@n3, Alpha=@n2, Translation="0,0,0", Rotation="20,0,0") {
    rotationMode=BMM_Additive
    rotationSpace=BCS_BoneSpace
    scaleMode=BMM_Additive
    scaleSpace=BCS_BoneSpace
    translationMode=BMM_Additive
    translationSpace=BCS_BoneSpace
    alphaBoolBlend {blendOption=Linear}
    alphaScaleBiasClamp {bMapRange=true, inRange {max=0, min=-40}, outRange {max=0, min=1}}
    boneToModify {boneName=root}
    rotation {pitch=20}
   }
   "Transform (Modify) Bone - Bone: dyn_pack_a_base"(Alpha=@n1, Translation="0,0,0", Rotation="0,15,0", Scale="1,1,1") {
    rotationMode=BMM_Additive
    rotationSpace=BCS_BoneSpace
    scaleMode=BMM_Additive
    scaleSpace=BCS_BoneSpace
    translationMode=BMM_Additive
    translationSpace=BCS_BoneSpace
    alphaBoolBlend {blendOption=Linear}
    alphaScaleBiasClamp {bMapRange=true, inRange {max=0, min=-50}, outRange {max=0, min=1}}
    boneToModify {boneName=dyn_pack_a_base}
    rotation {yaw=15}
   }
   n4 = Trail controller - Bone: dyn_pack_f!(
    ComponentPose=Local To Component(LocalPose=Use cached pose 'SavedPose_main'())
    Alpha=Get trailAlpha() -> (trailAlpha:double)
    RelaxationSpeedScale:float=Get trailRelation() -> (trailRelation:double)
   ) -> (Pose) {
    chainLength=6
    relaxationSpeedScale=3
    rotationLimits [
     {limitMax=<180 180 180>}
     {limitMax=<180 180 180>}
     {limitMax=<180 180 180>}
     {limitMax=<180 180 180>}
     {limitMax=<180 180 180>}
     {limitMax=<180 180 180>}
    ]
    rotationOffsets [<0 0 0>, <0 0 0>, <0 0 0>, <0 0 0>, <0 0 0>, <0 0 0>]
    trailBone {boneName=dyn_pack_f}
   }
   n5 = "Transform (Modify) Bone - Bone: dyn_pack_a_base"(ComponentPose=@n4, Alpha=1, Translation="0,0,0", Rotation="0,-8,0", Scale="1,1,1") {
    rotationMode=BMM_Additive
    rotationSpace=BCS_BoneSpace
    alphaBoolBlend {blendOption=Linear}
    boneToModify {boneName=dyn_pack_a_base}
    rotation {yaw=-8}
   }
   n6 = RigidBody(ComponentPose=Local To Component(LocalPose=Use cached pose 'SavedPose_main'())) {
    bClampLinearTranslationLimitToRefPose=true
    bOverrideWorldGravity=true
    overridePhysicsAsset=&/M_MED_Syko_Pack_Physics
    simulationSpace=BaseBoneSpace
    alphaBoolBlend {blendOption=Linear}
    alphaScaleBiasClamp {bInterpResult=true, outRange {min=0.5}}
    baseBoneRef {boneName=pack}
    componentLinearAccScale=<1 0.9 0.4>
    componentLinearVelScale=<1 0.9 0.4>
    overrideWorldGravity {z=-980}
   }
   n7 = "Transform (Modify) Bone - Bone: dyn_pack_a_base"(ComponentPose=@n6, Alpha=1, Translation="0,0,0", Rotation="0,-8,0", Scale="1,1,1") {
    rotationMode=BMM_Additive
    rotationSpace=BCS_BoneSpace
    alphaBoolBlend {blendOption=Linear}
    boneToModify {boneName=dyn_pack_a_base}
    rotation {yaw=-8}
   }
   Output Pose(Result=Blend Poses by bool(
     BlendPose_0=Component To Local(ComponentPose=@n5)
     BlendPose_1=Component To Local(ComponentPose=@n7)
     bActiveValue=Get isRigidBodyDisabled() -> (isRigidBodyDisabled:bool)
    ) {blendType=Linear}) {layerGroup=None}
   AnimGraphNode_SubInput_0 = Input Pose()
   n8 = "Transform (Modify) Bone - Bone: pack"(ComponentPose=Local To Component(LocalPose=@AnimGraphNode_SubInput_0), Alpha=0, Translation="0,0,0", Rotation="0,0,-9") {
    rotationMode=BMM_Additive
    rotationSpace=BCS_BoneSpace
    scaleMode=BMM_Additive
    scaleSpace=BCS_BoneSpace
    translationMode=BMM_Additive
    translationSpace=BCS_BoneSpace
    alphaBoolBlend {blendOption=Linear}
    boneToModify {boneName=pack}
    rotation {roll=-9}
   }
   n9 = "Transform (Modify) Bone - Bone: dyn_pack_a_offset"(ComponentPose=@n8, Alpha=1, Translation=Get floatingVec1() -> (floatingVec1:Vector), Rotation="0,0,0", Scale="1,1,1") {
    rotationMode=BMM_Additive
    rotationSpace=BCS_BoneSpace
    scaleMode=BMM_Additive
    scaleSpace=BCS_BoneSpace
    translationMode=BMM_Additive
    translationSpace=BCS_BoneSpace
    alphaBoolBlend {blendOption=Linear}
    boneToModify {boneName=dyn_pack_a_offset}
   }
   n10 = "Transform (Modify) Bone - Bone: dyn_pack_b_offset"(ComponentPose=@n9, Alpha=1, Translation=Get floatingVec3() -> (floatingVec3:Vector), Rotation="0,0,0", Scale="1,1,1") {
    rotationMode=BMM_Additive
    rotationSpace=BCS_BoneSpace
    scaleMode=BMM_Additive
    scaleSpace=BCS_BoneSpace
    translationMode=BMM_Additive
    translationSpace=BCS_BoneSpace
    alphaBoolBlend {blendOption=Linear}
    boneToModify {boneName=dyn_pack_b_offset}
   }
   n11 = "Transform (Modify) Bone - Bone: dyn_pack_c_offset"(ComponentPose=@n10, Alpha=1, Translation=Get floatingVec2() -> (floatingVec2:Vector), Rotation="0,0,0", Scale="1,1,1") {
    rotationMode=BMM_Additive
    rotationSpace=BCS_BoneSpace
    scaleMode=BMM_Additive
    scaleSpace=BCS_BoneSpace
    translationMode=BMM_Additive
    translationSpace=BCS_BoneSpace
    alphaBoolBlend {blendOption=Linear}
    boneToModify {boneName=dyn_pack_c_offset}
   }
   n12 = "Transform (Modify) Bone - Bone: dyn_pack_d_offset"(ComponentPose=@n11, Alpha=1, Translation=Get floatingVec1(), Rotation="0,0,0", Scale="1,1,1") {
    rotationMode=BMM_Additive
    rotationSpace=BCS_BoneSpace
    scaleMode=BMM_Additive
    scaleSpace=BCS_BoneSpace
    translationMode=BMM_Additive
    translationSpace=BCS_BoneSpace
    alphaBoolBlend {blendOption=Linear}
    boneToModify {boneName=dyn_pack_d_offset}
   }
   n13 = "Transform (Modify) Bone - Bone: dyn_pack_e_offset"(ComponentPose=@n12, Alpha=1, Translation=Get floatingVec3(), Rotation="0,0,0", Scale="1,1,1") {
    rotationMode=BMM_Additive
    rotationSpace=BCS_BoneSpace
    scaleMode=BMM_Additive
    scaleSpace=BCS_BoneSpace
    translationMode=BMM_Additive
    translationSpace=BCS_BoneSpace
    alphaBoolBlend {blendOption=Linear}
    boneToModify {boneName=dyn_pack_e_offset}
   }
   n14 = "Transform (Modify) Bone - Bone: dyn_pack_f_offset"(ComponentPose=@n13, Alpha=1, Translation=Get floatingVec2(), Rotation="0,0,0", Scale="1,1,1") {
    rotationMode=BMM_Additive
    rotationSpace=BCS_BoneSpace
    scaleMode=BMM_Additive
    scaleSpace=BCS_BoneSpace
    translationMode=BMM_Additive
    translationSpace=BCS_BoneSpace
    alphaBoolBlend {blendOption=Linear}
    boneToModify {boneName=dyn_pack_f_offset}
   }
   n15 = "Transform (Modify) Bone - Bone: pelvis"(
    ComponentPose=@n14
    Alpha=1
    Translation=Get pelvis_tran() -> (pelvis_tran:Vector)
    Rotation=Get pelvis_rot() -> (pelvis_rot:Rotator)
    Scale="1,1,1"
   ) {
    rotationMode=BMM_Replace
    rotationSpace=BCS_WorldSpace
    scaleSpace=BCS_BoneSpace
    translationMode=BMM_Replace
    translationSpace=BCS_WorldSpace
    alphaBoolBlend {blendOption=Linear}
    alphaScaleBiasClamp {inRange {max=0, min=-40}, outRange {max=0, min=1}}
    boneToModify {boneName=pelvis}
   }
   n16 = "Transform (Modify) Bone - Bone: spine_02"(
    ComponentPose=@n15
    Alpha=1
    Translation=Get spine02_tran() -> (spine02_tran:Vector)
    Rotation=Get spine02_rot() -> (spine02_rot:Rotator)
    Scale="1,1,1"
   ) {
    rotationMode=BMM_Replace
    rotationSpace=BCS_WorldSpace
    scaleSpace=BCS_BoneSpace
    translationMode=BMM_Replace
    translationSpace=BCS_WorldSpace
    alphaBoolBlend {blendOption=Linear}
    alphaScaleBiasClamp {inRange {max=0, min=-40}, outRange {max=0, min=1}}
    boneToModify {boneName=spine_02}
   }
   n17 = "Transform (Modify) Bone - Bone: thigh_l"(
    ComponentPose=@n16
    Alpha=1
    Translation=Get thigh_l_tran() -> (thigh_l_tran:Vector)
    Rotation=Get thigh_l_rot() -> (thigh_l_rot:Rotator)
    Scale="1,1,1"
   ) {
    rotationMode=BMM_Replace
    rotationSpace=BCS_WorldSpace
    scaleSpace=BCS_BoneSpace
    translationMode=BMM_Replace
    translationSpace=BCS_WorldSpace
    alphaBoolBlend {blendOption=Linear}
    alphaScaleBiasClamp {inRange {max=0, min=-40}, outRange {max=0, min=1}}
    boneToModify {boneName=thigh_l}
   }
   n18 = "Transform (Modify) Bone - Bone: thigh_r"(
    ComponentPose=@n17
    Alpha=1
    Translation=Get thigh_r_tran() -> (thigh_r_tran:Vector)
    Rotation=Get thigh_r_rot() -> (thigh_r_rot:Rotator)
    Scale="1,1,1"
   ) {
    rotationMode=BMM_Replace
    rotationSpace=BCS_WorldSpace
    scaleSpace=BCS_BoneSpace
    translationMode=BMM_Replace
    translationSpace=BCS_WorldSpace
    alphaBoolBlend {blendOption=Linear}
    alphaScaleBiasClamp {inRange {max=0, min=-40}, outRange {max=0, min=1}}
    boneToModify {boneName=thigh_r}
   }
   n19 = "Transform (Modify) Bone - Bone: root"(ComponentPose=@n18, Translation="0,0,0", Rotation=Get backpack_rotate()) {rotationMode=BMM_Additive, rotationSpace=BCS_BoneSpace, alphaBoolBlend {blendOption=Linear}, boneToModify {boneName=root}}
   Save cached pose 'SavedPose_main'(Pose=Component To Local(ComponentPose=@n19)) {cacheName=SavedPose_main}
   Get movingBwd() -> (movingBwd:double)
  ]}
 EventGraph {nodes [
   Event BlueprintUpdateAnimation()
   > n1 = Sequence!(execute) -> (then_0, then_1, then_2:Exec, then_3:Exec, then_4:Exec, then_5:Exec, then_6:Exec, then_7:Exec, then_8:Exec)
   > n2 = Sequence!(execute) -> (then_0, then_1, then_2)
   > n3 = RandomMoveFNC!(execute, self, frequency:double=2, amp:double=0.42) -> (then, tran_output:double)
   > Set floatingVec1(floatingVec1:Vector=MakeVector(X=@n3, Y=0, Z=@n3)) -> (Output_Get:Vector)
   n4 = RandomMoveFNC(execute=@n2.then_1, frequency=-3.5, amp=-0.22)
   > Set floatingVec2(floatingVec2:Vector=MakeVector(X=@n4, Y=0, Z=@n4)) -> (Output_Get:Vector)
   n5 = RandomMoveFNC(execute=@n2.then_2, frequency=2.5, amp=0.32)
   > Set floatingVec3(floatingVec3:Vector=MakeVector(X=@n5, Y=0, Z=@n5)) -> (Output_Get:Vector)
   n6 = GetOwnerDirection!(self:Custom Character Part Anim Instance Object Reference) -> (ReturnValue:float)
   n7 = Set movingBwd(execute=@n1.then_4, movingBwd:double=MapRangeClamped(Value="Absolute (Float)"(A=@n6), InRangeA=130, InRangeB=180, OutRangeA=1, OutRangeB=0)) -> (Output_Get:double)
   > n8 = Branch(Condition=Get bIsSkyDiving())
   > Set trailAlpha(trailAlpha:double=MapRangeClamped(Value=@n7, InRangeA=0, InRangeB=1, OutRangeB=0.2)) -> (Output_Get:double)
   > Set trailRelation(trailRelation:double=2) -> (Output_Get:double)
   Set trailAlpha(execute=@n8.else, trailAlpha=@n7)
   > Set trailRelation(trailRelation=3)
   n9 = GetPartSkeletalMeshComponent()
   Is Valid(exec=@n1.then_6, InputObject=@n9)
   n10 = GetAttachParent!(self:Scene Component Object Reference=@n9) -> (ReturnValue:Scene Component Object Reference)
   > Is Valid(InputObject=@n10)
   > n11 = Sequence!(execute) -> (then_0, then_1, then_2, then_3:Exec, then_4:Exec, then_5:Exec, then_6:Exec, then_7:Exec)
   n12 = BreakTransform(InTransform=GetSocketTransform(self=@n10, InSocketName=pelvis, TransformSpace=RTS_World))
   > Set pelvis_tran(pelvis_tran:Vector=@n12) -> (Output_Get:Vector)
   Set pelvis_rot(execute=@n11.then_1, pelvis_rot:Rotator=@n12.Rotation) -> (Output_Get:Rotator)
   n13 = BreakTransform(InTransform=GetSocketTransform(self=@n10, InSocketName=spine_02, TransformSpace=RTS_World))
   Set spine02_tran(execute=@n11.then_2, spine02_tran:Vector=@n13) -> (Output_Get:Vector)
   Set spine02_rot(execute=@n11.then_3, spine02_rot:Rotator=@n13.Rotation) -> (Output_Get:Rotator)
   n14 = BreakTransform(InTransform=GetSocketTransform(self=@n10, InSocketName=thigh_l, TransformSpace=RTS_World))
   Set thigh_l_tran(execute=@n11.then_4, thigh_l_tran:Vector=@n14) -> (Output_Get:Vector)
   Set thigh_l_rot(execute=@n11.then_5, thigh_l_rot:Rotator=@n14.Rotation) -> (Output_Get:Rotator)
   n15 = BreakTransform(InTransform=GetSocketTransform(self=@n10, InSocketName=thigh_r, TransformSpace=RTS_World))
   Set thigh_r_tran(execute=@n11.then_6, thigh_r_tran:Vector=@n15) -> (Output_Get:Vector)
   Set thigh_r_rot(execute=@n11.then_7, thigh_r_rot:Rotator=@n15.Rotation) -> (Output_Get:Rotator)
   n16 = Cast To FortPlayerPawn(execute=@n1.then_3, Object=TryGetPawnOwner())
   > Set IsMeleeAnim(IsMeleeAnim:bool=IsPlayingMeleeAnim!(self:Fort Player Animation Interface Object Reference=GetAnimationInterface!(self:Fort Player Pawn Object Reference=@n16) -> (ReturnValue:Fort Player Animation Interface Object Reference)) -> (ReturnValue:bool)) -> (Output_Get:bool)
   Set IsMeleeAnim(execute=@n16.CastFailed, IsMeleeAnim=false)
   n17 = GetOwnerSkeletalMeshComponent()
   n18 = BreakRotator(InRot=BreakTransform(InTransform=GetDeltaTransformFromRefPose(self=@n17, BoneName=spine_05, BaseName=spine_01)).Rotation)
   Set backpack_rotate(execute=@n1.then_7, backpack_rotate=MakeRotator(Roll=MapRangeClamped(Value=@n18.Yaw, InRangeA=-5, InRangeB=-30, OutRangeB=30)))
   n19 = Get CurrentLOD()
   n20 = Get bRigidBodyEnabled()
   Set isRigidBodyDisabled(
    execute=@n1.then_5
    isRigidBodyDisabled:bool=OR Boolean(A=NOT Boolean(A=@n20), B="integer > integer"(A=@n19, B=1) {is=title K2Node_CallFunction})
   ) -> (Output_Get:bool)
   n21 = Branch(Condition=Get IsMeleeAnim() -> (IsMeleeAnim:bool))
   n22 = Get torsoTwist()
   > Set pickaxeRigidBlend(pickaxeRigidBlend:double=MapRangeClamped(Value="Absolute (Float)"(A=@n22), InRangeA=0, InRangeB=30, OutRangeA=1, OutRangeB=0)) -> (Output_Get:double)
   Set pickaxeRigidBlend(execute=@n21.else, pickaxeRigidBlend=1)
   n23 = GetSocketTransform(InSocketName=spine_05, TransformSpace=RTS_World)
   n24 = GetSocketTransform(InSocketName=spine_01, TransformSpace=RTS_World)
   n25 = BreakRotator(InRot=BreakTransform(InTransform=MakeRelativeTransform!(A:"Transform (by ref)"=@n24, RelativeTo:"Transform (by ref)"=@n23) -> (ReturnValue:Transform)).Rotation)
   Set torsoTwist(torsoTwist:double=@n25) -> (Output_Get:double)
   > Set torsoBend(torsoBend:double=@n25.Yaw) -> (Output_Get:double)
   Set jump_alpha(jump_alpha:double=MapRangeClamped(Value=Get PawnVelocityZ(), InRangeA=-700, InRangeB=0, OutRangeA=1, OutRangeB=0)) -> (Output_Get:double)
  ]}
 RandomMoveFNC {nodes [
   n1 = RandomMoveFNC!() -> (then, frequency:double=1, amp:double=1) {is=title K2Node_FunctionEntry}
   n2 = float * float(A=@n1, B=GetGameTimeInSeconds())
   n3 = float * float(A="Sin (Radians)"(A=@n2), B=@n1.amp)
   > Return Node!(execute, tran_output:double=@n3)
   PerlinNoise1D!(Value:float=@n2) -> (ReturnValue:float)
   PerlinNoise1D(Value=@n3)
  ]}
]
variables [
 backpack_rotate {type=Rotator}
 floatingVec1 {type=Vector, default=<-0.2785 0 0>}
 floatingVec2 {type=Vector, default=<-0.2962 0 0>}
 floatingVec3 {type=Vector, default=<0.0329 0 0>}
 IsMeleeAnim {type=boolean}
 isRigidBodyDisabled {type=boolean}
 jump_alpha {type=number}
 movingBwd {type=number}
 pelvis_rot {type=Rotator}
 pelvis_tran {type=Vector}
 pickaxeRigidBlend {type=number}
 spine02_rot {type=Rotator}
 spine02_tran {type=Vector}
 thigh_l_rot {type=Rotator}
 thigh_l_tran {type=Vector}
 thigh_r_rot {type=Rotator}
 thigh_r_tran {type=Vector}
 torsoBend {type=number}
 torsoTwist {type=number}
 trailAlpha {type=number}
 trailRelation {type=number}
]
